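# Sound cue for volcano ambience in Athena, playing thunder or earth-movement one-shots by stage on the scripted events sound class.
# A Switch node reads the integer parameter Stage to pick among four inputs, two of which share the Thunder_03 wave, alongside Thunder_02 and Earth_Move_01.
# Attenuation override disables distance attenuation and sets manual reverb send, while concurrency override caps simultaneous plays at two and prevents new ones.
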SoundCue /FNGameplayCues/Athena/Sounds/Ambient/Volcano/Amb_Volcano_Snow_Cue
root=Switch_2
nodes [
 {id=Switch_2, children [WavePlayer_2,WavePlayer_2,WavePlayer_1,WavePlayer_4], properties {intParameterName=Stage}}
 {id=WavePlayer_1, properties {soundWaveAssetPtr=&/Amb_Volcano_Thunder_02}}
 {id=WavePlayer_2, properties {soundWaveAssetPtr=&/Amb_Volcano_Thunder_03}}
 {id=WavePlayer_4, properties {soundWaveAssetPtr=&/Amb_Volcano_Earth_Move_01}}
]
properties {
 bOverrideAttenuation=true
 bOverrideConcurrency=true
 maxDistance=2097152
 soundClassObject=/Game/Sounds/FortSoundClassesAndModes/Fort_Events_Scripted
 volumeMultiplier=0.3509
 attenuationOverrides {
  bAttenuate=false
  falloffDistance=100000
  manualReverbSendLevel=15
  reverbSendMethod=Manual
  stereoSpread=4000
  attenuationShapeExtents=<2000 1500>
 }
 concurrencyOverrides {maxCount=2, resolutionRule=PreventNew}
}
